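# Test Default Lit material for the Item Shop VaultWorld UI, built around the shared MF_UI_MPItemShop_Setup function.
# Screen-space UVs scaled to fill a 1920x1080 reference are shared through a named reroute.
# A blurred grayscale noise texture is partly desaturated and fed as Color into the setup function, with a FinalBrightness parameter controlling brightness.
# A smoothstepped blue-to-purple linear gradient is also built but feeds nothing.
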Material /ItemShop/UI/Art/VaultWorld_Assets/M_UI_VaultWorld_Test
graph {
 properties {pixelDepthOffsetMode=PDOM_Legacy, usedShadingModels=Default Lit}
 nodes [
  n1 = MaterialFunctionCall!(ScreenUV, TextureSize=Constant2Vector() {g=1080, r=1920}, ScalePivotPoint) {materialFunction=/FortUILibrary/MaterialFunctions/Basics/MF_UI_ScreenSpaceUVScaling}
  n2 = Constant()
  n3 = Constant() {r=1}
  LinearInterpolate(
   A=Constant3Vector() {constant=<rgba 0.0003 0.0497 0.4983 1>}
   B=Constant3Vector() {constant=<rgba 0.4526 0.1785 1 1>}
   Alpha=SmoothStep(Value=MaterialFunctionCall(UVs=@n1.Scaled to Fill, GradientStart=@n2, GradientEnd=@n3) {materialFunction=&MF_UI_LinearGradient}.Result)
  )
  n4 = NamedRerouteUsage()
  MaterialFunctionCall!(
   Color=Desaturation(
    None=TextureSample(UVs=@n4) {samplerType=SAMPLERTYPE_Grayscale, texture=/FortUILibrary/UIKit/Buttons/Textures/LowResBlurredNoise}.RGB
    Fraction=Constant() {r=0.35}
   )
   Color Brightness=ScalarParameter() {defaultValue=0.75, parameterName=FinalBrightness}
   Floor Radius
   Floor Density
   Floor Metallic
   Floor Specular
   Floor Roughness
  ) {materialFunction=/Game/UI/Foundation/Environment/Shell/Materials/MaterialFunctions/MF_UI_MPItemShop_Setup}
  NamedRerouteDeclaration(None=@n1.Scaled to Fill) {name=UVFIll, nodeColor=<rgba 1 0 0.1406 1>}
 ]
}
